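# Flipbook face material for the RecipeWhole character's head, driven by Animation Blueprint eye and brow index parameters.
# MF_FlipbookFaceRecipeWhole and an anime eye-animation function supply color and feature masks, composited between character-parent PreFX and PostFX functions.
# Certain eye and brow indices, when left and right match, enable rescaled, offset second-UV mask samples that black out base color.
# HasPoseChanged adjusts temporal responsiveness within masked face regions.
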Material /BRCosmetics/Characters/Player/Male/Medium/Bodies/M_MED_RecipeWhole/Materials/M_FlipbookFace_RecipeWhole
graph {
 properties {
  bUseMaterialAttributes=true
  bUsedWithMeshDeformer=true
  bUsedWithMorphTargets=true
  bUsedWithSkeletalMesh=true
  pixelDepthOffsetMode=PDOM_Legacy
  shadingModel=MSM_FromMaterialExpression
  usedShadingModels=Default Lit | Subsurface | Clear Coat | Subsurface Profile | Cloth
 }
 nodes [
  Constant() {r=1}
  n1 = NamedRerouteUsage()
  n2 = TextureSample(UVs=NamedRerouteUsage()) {constCoordinate=1, texture=/BRCosmetics/Characters/Player/Male/Medium/Bodies/M_MED_RecipeWhole/Textures/T_RecipeWhole_Head_Flipbook_D}
  n3 = Constant()
  n4 = If(A=@n2.R, "A > B"=Constant() {r=1}, "A == B"=@n3, "A < B"=@n3) {constB=0.1}
  n5 = NamedRerouteUsage()
  n6 = TextureSample(UVs=NamedRerouteUsage()) {constCoordinate=1, texture=/BRCosmetics/Characters/Player/Male/Medium/Bodies/M_MED_RecipeWhole/Textures/T_RecipeWhole_Head_Flipbook_EyeMask}
  n7 = Multiply(A=@n6.R) {constB=2}
  n8 = LinearInterpolate(A=LinearInterpolate(A=NamedRerouteUsage(), B=@n5, Alpha=Saturate(None=@n7)), B=@n2.RGB, Alpha=@n5)
  n9 = LinearInterpolate(
   A=LinearInterpolate(A=Multiply(A=@n4, B=@n8), B=NamedRerouteUsage(), Alpha=NamedRerouteUsage())
   B=NamedRerouteUsage()
   Alpha=Multiply(A=NamedRerouteUsage(), B=NamedRerouteUsage())
  )
  n10 = NamedRerouteUsage()
  n11 = TextureCoordinate() {coordinateIndex=1}
  n12 = Constant() {r=0.112}
  n13 = Constant4Vector() {constant=<rgba -0.136 2.0602 0 1>}
  n14 = TextureSample(UVs=Add(
    A=MaterialFunctionCall(UVs=@n11, Texture Scale=@n12) {materialFunction=&ScaleUVsByCenter}.UVs
    B=ComponentMask(None=@n13.RGB) {g=true, r=true}
   )) {
   constCoordinate=1
   samplerSource=SSM_Clamp_WorldGroupSettings
   samplerType=SAMPLERTYPE_Masks
   texture=/BRCosmetics/Characters/Player/Male/Medium/Bodies/M_MED_RecipeWhole/Textures/T_RecipeWhole_Head_Flipbook_EyeBrowsMask
  }
  n15 = TextureCoordinate() {coordinateIndex=1}
  n16 = ScalarParameter() {defaultValue=0.072, parameterName=Index8_TexScale}
  n17 = VectorParameter() {parameterName=Index8_UVOffset, defaultValue=<rgba -0.136 2.4741 0 1>}
  n18 = TextureSample(UVs=Add(
    A=MaterialFunctionCall(UVs=@n15, Texture Scale=@n16) {materialFunction=&ScaleUVsByCenter}.UVs
    B=ComponentMask(None=@n17.RGB) {g=true, r=true}
   )) {
   constCoordinate=1
   samplerSource=SSM_Clamp_WorldGroupSettings
   samplerType=SAMPLERTYPE_Masks
   texture=&T_RecipeWhole_Head_Flipbook_EyeBrowsMask
  }
  n19 = Max(A=LinearInterpolate(B=@n14.G, Alpha=NamedRerouteUsage()), B=LinearInterpolate(B=@n18.R, Alpha=NamedRerouteUsage()))
  n20 = Constant() {r=1}
  n21 = If(A=Max(A=NamedRerouteUsage(), B=NamedRerouteUsage()), "A > B"=@n20, "A == B"=@n20, "A < B"=Constant()) {constB=1}
  n22 = LinearInterpolate(A=LinearInterpolate(A=LinearInterpolate(A=@n9, Alpha=@n10) {constB=0}, Alpha=Multiply(A=@n19, B=@n21)) {constB=0}, Alpha=NamedRerouteUsage()) {constB=0}
  n23 = MaterialFunctionCall() {materialFunction=&MF_CharacterParent_2023_PreFX}
  MaterialFunctionCall(
   Material=MaterialFunctionCall!(Material=SetMaterialAttributes!(MaterialAttributes=@n1, Base Color=@n22, World Position Offset) {attributeSetTypes [#1, #2]}) {materialFunction=/Game/Global/Materials/Character_Parent_2023/Functions/BaseFN_CharacterOpacity_2023}.Result
   M=@n23.M
   Pre Skinned Local Position=@n23.M
   Distance Blend=@n23.M
   Fresnel=@n23.M
  ) {materialFunction=&MF_CharacterParent_2023_PostFX}
  n24 = MaterialFunctionCall!(Mouth UV, Eye UV, Brow UV, Material=@n23.Result) {materialFunction=&/MF_FlipbookFaceRecipeWhole}
  n25 = GetMaterialAttributes(None=@n24.Material) {attributeGetTypes [#1, #2]}
  NamedRerouteDeclaration(None=@n25.Base Color) {name=InBaseColor, nodeColor=<rgba 0 1 0.4141 1>}
  n26 = TextureSample(UVs=NamedRerouteUsage()) {constCoordinate=1, texture=&T_RecipeWhole_Head_Flipbook_EyeMask}
  n27 = Multiply(A=@n26.R) {constB=2}
  NamedRerouteDeclaration(None=MaterialFunctionCall!(
    Color Lit=NamedRerouteUsage()
    Color Shaded
    Use UV2=StaticBoolParameter() {defaultValue=true, group=01_EyeAnimation_Parameters, parameterName=Use UV2 for Eye Animation, sortPriority=-30}
    Mask=Saturate(None=@n27)
   ) {materialFunction=&/MF_AnimeEyeAnimation_RecipeWhole_Head}.Lit) {name=OutBaseColor, nodeColor=<rgba 0 0.2266 1 1>}
  n28 = Constant4Vector() {constant=<rgba -0.01 -0.02 0 1>}
  NamedRerouteDeclaration(None=Add(A=TextureCoordinate() {coordinateIndex=1}, B=ComponentMask(None=@n28.RGB) {g=true, r=true})) {name=UVCenterOffset, nodeColor=<rgba 0 0.8125 1 1>}
  NamedRerouteDeclaration(None=@n25.MaterialAttributes) {name=In_MA, nodeColor=<rgba 0 0.2734 1 1>}
  n29 = ScalarParameter() {
   desc="Driven by Animation Blueprint.  Do not set manually."
   group=Flipbook 05 Indices
   parameterName=flipbook_face_R_eye_index
   sortPriority=11
  }
  n30 = ScalarParameter() {
   desc="Driven by Animation Blueprint.  Do not set manually."
   group=Flipbook 05 Indices
   parameterName=flipbook_face_L_eye_index
   sortPriority=10
  }
  n31 = Constant()
  n32 = Constant()
  NamedRerouteDeclaration(None=If(A=If(A=@n29, B=@n30, "A > B"=@n31, "A == B"=@n29, "A < B"=@n31) {constB=7}, "A > B"=@n32, "A == B"=Constant() {r=1}, "A < B"=@n32) {constB=4}) {name=Eye_index4, nodeColor=<rgba 0 0.9766 1 1>}
  n33 = ScalarParameter() {
   desc="Driven by Animation Blueprint.  Do not set manually."
   group=Flipbook 05 Indices
   parameterName=flipbook_face_R_eye_index
   sortPriority=11
  }
  n34 = ScalarParameter() {
   desc="Driven by Animation Blueprint.  Do not set manually."
   group=Flipbook 05 Indices
   parameterName=flipbook_face_L_eye_index
   sortPriority=10
  }
  n35 = Constant()
  n36 = Constant()
  NamedRerouteDeclaration(None=If(A=If(A=@n33, B=@n34, "A > B"=@n35, "A == B"=@n33, "A < B"=@n35) {constB=7}, "A > B"=@n36, "A == B"=Constant() {r=1}, "A < B"=@n36) {constB=5}) {name=Eye_index5, nodeColor=<rgba 0 0.9766 1 1>}
  n37 = TextureCoordinate() {coordinateIndex=1}
  n38 = ScalarParameter() {defaultValue=0.072, parameterName=Index5_TexScale}
  n39 = VectorParameter() {parameterName=Index5_UVOffset, defaultValue=<rgba -0.136 2.8544 0 1>}
  n40 = TextureSample(UVs=Add(
    A=MaterialFunctionCall(UVs=@n37, Texture Scale=@n38) {materialFunction=&ScaleUVsByCenter}.UVs
    B=ComponentMask(None=@n39.RGB) {g=true, r=true}
   )) {
   constCoordinate=1
   samplerSource=SSM_Clamp_WorldGroupSettings
   samplerType=SAMPLERTYPE_Masks
   texture=&T_RecipeWhole_Head_Flipbook_EyeBrowsMask
  }
  n41 = TextureCoordinate() {coordinateIndex=1}
  n42 = ScalarParameter() {defaultValue=0.072, parameterName=Index4_TexScale}
  n43 = VectorParameter() {parameterName=Index4_UVOffset, defaultValue=<rgba -0.136 1.6784 0 1>}
  n44 = TextureSample(UVs=Add(
    A=MaterialFunctionCall(UVs=@n41, Texture Scale=@n42) {materialFunction=&ScaleUVsByCenter}.UVs
    B=ComponentMask(None=@n43.RGB) {g=true, r=true}
   )) {
   constCoordinate=1
   samplerSource=SSM_Clamp_WorldGroupSettings
   samplerType=SAMPLERTYPE_Masks
   texture=&T_RecipeWhole_Head_Flipbook_EyeBrowsMask
  }
  n45 = Max(A=LinearInterpolate(B=@n40.R, Alpha=NamedRerouteUsage()), B=LinearInterpolate(B=@n44.R, Alpha=NamedRerouteUsage()))
  n46 = Constant() {r=1}
  n47 = If(A=Max(A=NamedRerouteUsage(), B=NamedRerouteUsage()), "A > B"=@n46, "A == B"=@n46, "A < B"=Constant()) {constB=1}
  NamedRerouteDeclaration(None=LinearInterpolate(A=Multiply(A=@n45, B=@n47), Alpha=ScalarParameter() {group=ForRender, parameterName=ForRender}) {constB=0}) {name=EyeIndex4_5, nodeColor=<rgba 1 0 0.1172 1>}
  NamedRerouteDeclaration(None=@n24.EyesColor_2) {name=EyeColor2_Mask, nodeColor=<rgba 0.7344 0 1 1>}
  NamedRerouteDeclaration(None=@n24.RemoveEyeWide) {name=RemoveEyeWide_1, nodeColor=<rgba 0 0.3672 1 1>}
  NamedRerouteDeclaration(None=@n24.BrowsVertexMask) {name=BrowVertexMask, nodeColor=<rgba 1 0 0 1>}
  NamedRerouteDeclaration(None=@n24.MouthAlpha) {name=MouthAlpha, nodeColor=<rgba 1 0.5859 0 1>}
  n48 = ScalarParameter() {
   desc="Driven by Animation Blueprint.  Do not set manually."
   group=Flipbook 05 Indices
   parameterName=Flipbook_face_R_brow_index
   sortPriority=1
  }
  n49 = ScalarParameter() {
   desc="Driven by Animation Blueprint.  Do not set manually."
   group=Flipbook 05 Indices
   parameterName=Flipbook_face_L_brow_index
   sortPriority=0
  }
  n50 = Constant()
  n51 = Constant()
  NamedRerouteDeclaration(None=If(A=If(A=@n48, B=@n49, "A > B"=@n50, "A == B"=@n48, "A < B"=@n50) {constB=7}, "A > B"=@n51, "A == B"=Constant() {r=1}, "A < B"=@n51) {constB=7}) {name=brow_index7_01, nodeColor=<rgba 0 0.9766 1 1>}
  n52 = ScalarParameter() {
   desc="Driven by Animation Blueprint.  Do not set manually."
   group=Flipbook 05 Indices
   parameterName=Flipbook_face_R_brow_index
   sortPriority=1
  }
  n53 = ScalarParameter() {
   desc="Driven by Animation Blueprint.  Do not set manually."
   group=Flipbook 05 Indices
   parameterName=Flipbook_face_L_brow_index
   sortPriority=0
  }
  n54 = Constant()
  n55 = Constant()
  NamedRerouteDeclaration(None=If(A=If(A=@n52, B=@n53, "A > B"=@n54, "A == B"=@n52, "A < B"=@n54) {constB=7}, "A > B"=@n55, "A == B"=Constant() {r=1}, "A < B"=@n55) {constB=8}) {name=brow_index8_01, nodeColor=<rgba 0 0.9766 1 1>}
  NamedRerouteUsage()
  ScalarParameter() {
   desc="Driven by Animation Blueprint.  Do not set manually."
   group=Flipbook 05 Indices
   parameterName=Flipbook_face_L_brow_index
   sortPriority=0
  }
  n56 = Max(A=@n26.R, B=Max(A=NamedRerouteUsage(), B=NamedRerouteUsage()))
  n57 = ScalarParameter() {desc="Set by base head AnimBP if \"IsFlipbookFace\" is true", parameterName=HasPoseChanged}
  TemporalResponsivenessOutput!(None=LinearInterpolate(Alpha=Multiply(A=@n56, B=@n57)) {constA=0.4})
 ]
}
